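# Unlit translucent, two-sided particle material for radial smoke on the Bunker Man pickaxe impact, usable on Niagara and particle sprites.
# Particle color RGB feeds an emissive-adjustment function.
# Particle alpha and a soft smoke mask are multiplied by a dissolve term, which blends two curves from a dynamic parameter through a tiling smoke-pattern texture.
# Depth fade softens the resulting opacity, except on the mobile shading path.
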Material /Game/Weapons/FORT_Melee/Pickaxe_BunkerMan/FX/M_Melee_Bunker_Man_impact_RadialSmoke
graph {
 properties {
  bUsedWithNiagaraSprites=true
  bUsedWithParticleSprites=true
  blendMode=BLEND_Translucent
  pixelDepthOffsetMode=PDOM_Legacy
  shadingModel=MSM_Unlit
  twoSided=true
  usedShadingModels=Unlit
 }
 nodes [
  n1 = ParticleColor()
  MaterialFunctionCall(Emissive=@n1.RGB) {materialFunction=&MF_FX_EmissiveAdjustments}
  n2 = Multiply(A=DynamicParameter() {paramNames ["Dissovle(0-1)", Param2, Param3, Param4]}."Dissovle(0-1)", B=Constant() {r=2})
  n3 = Multiply(A=@n2) {constB=-1.2}
  n4 = OneMinus(None=@n2)
  n5 = Add(A=@n2)
  n6 = Constant() {r=2}
  n7 = OneMinus(None=Subtract(A=@n5, B=@n6))
  n8 = Multiply(A=TextureCoordinate(), B=Constant2Vector() {g=1, r=1})
  n9 = Multiply(
   A=Multiply(A=@n1.A, B=TextureSample() {texture=/Game/Effects/Fort_Effects/Textures/T_Smoke_Soft}.R)
   B=LinearInterpolate(
    A=Add(A=@n3, B=@n4)
    B=If(A=@n5, B=@n6, "A > B"=@n7, "A == B"=@n5, "A < B"=@n5)
    Alpha=TextureSample(UVs=@n8) {texture=/Game/Global/Textures/Default/Tileable_Patterns/T_Blowing_Smoke_Pattern2_Tile_DISP}.R
   )
  )
  ShadingPathSwitch(Default=DepthFade(Opacity=@n9, FadeDistance=Constant() {r=20}), Mobile=@n9) {desc=Migrated from FeatureLevelSwitch}
 ]
}
